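; Emperor quest phone booth hiding prop that only players with quest tags can enter; others see "Locked".
; Hiding-player count changes swing the door, animate blinds through custom primitive data, and apply gameplay effects to the last hiding player, including one granting the ability that runs the booth sequence.
; An outfit-change event fires repeated wobble gameplay cues.
; The booth silently removes itself when its playlist data row disables it, and drops glass windows on mobile.
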(event Collision|EventActorBeginOverlap
  (params (OtherActor :type Actor%20Object%20Reference))
  (|CallParentFunction OtherActor) ; This node is disabled and will not be called.\nDrag off pins to build functionality.
)

(event EventTick
  (params (DeltaSeconds :type Float%20%28single-precision%29))
  (|CallParentFunction DeltaSeconds) ; This node is disabled and will not be called.\nDrag off pins to build functionality.
)

(event Custom|ChangingOutFitWobble
  (bind _returnvalue (Math|Random|RandomFloatinRange 6e+01 135.0))
  (bind _wobblelocationoffset (Variables|Default|GetWobbleLocationOffset))
  (bind _y (.y _wobblelocationoffset))
  (bind _x (.x _wobblelocationoffset))
  (bind _returnvalue_2 (Math|Vector|MakeVector _x _y _returnvalue))
  (bind _returnvalue_3 (Math|Random|RandomUnitVector))
  (bind _self self)
  (bind _returnvalue_4 (* _returnvalue_3 0.2))
  (bind _returnvalue_5 (Ability|GameplayCue|MakeGameplayCueParameters :NormalizedMagnitude 1e+02 :RawMagnitude 0.0 :Location _returnvalue_2 :Normal _returnvalue_4 :Instigator _self :GameplayEffectLevel 1 :AbilityLevel 1 :bReplicateLocationWhenUsingMinimalRepProxy false))
  (GameplayCue|ExecuteGameplayCue%28Burst%29Local self (Variables|Default|GetGC_Wobble_0) _returnvalue_5)
  (Utilities|FlowControl|Delay 0.4)
  (GameplayCue|ExecuteGameplayCue%28Burst%29Local self (Variables|Default|GetGC_Wobble_0) _returnvalue_5)
  (Utilities|FlowControl|Delay 0.8)
  (GameplayCue|ExecuteGameplayCue%28Burst%29Local self (Variables|Default|GetGC_Wobble_0) _returnvalue_5))

(event Custom|DoorSwingAnim
  (Collision|SetCollisionResponsetoChannel (Variables|Default|GetSM_Telephone_Booth_DoorGlass) "ECC_Camera" "ECR_Ignore")
  (bind (_direction _float) (Timeline|DoorSwing PlayFromStart)
    (:Update
      (Transformation|SetRelativeRotation (Variables|Default|GetDoor) (Math|Rotator|Lerp(Rotator) "0,90.000000, 0" "0,-30.000000, 0" _float) false false))
    (:Finished
      (Components|Activation|Activate (Variables|Default|GetNS_PhantomBooth_DoorSlam) true))))

(event EventBeginPlay
  (bind _staticmeshcomponent (Variables|Debug|GetStaticMeshComponent))
  (|CallParentFunction)
  ;[1 1 1] Reset Blinds Material
  (Rendering|Material|SetCustomPrimitiveDataFloat _staticmeshcomponent 5 1.0)
  (Rendering|Material|SetCustomPrimitiveDataFloat _staticmeshcomponent 6 0.0)
  ;[1 1 1] Check If Enabled
  (Class|FortAsyncActionCurrentPlaylistReady|PlaylistDataReadyAsync
    (:OnReady
      (if (not (Class|AthenaFunctionLibrary|RowtoBool (Variables|Default|GetEnabledRow)))
        (Gameplay|SilentDie false)))))

(event EventHidingPlayerCountChanged
  (bind _self self)
  (bind _returnvalue (Ability|GetFortAbilitySystemComponent _self))
  (bind _self_2 self)
  (|CallParentFunction)
  (Utilities|FlowControl|Sequence
    (:then_0
      (CallFunction|DoorSwingAnim)
      (Networking|FlushNetDormancy)
      (Utilities|IsValid (Utilities|Array|Get(acopy) (Variables|Default|GetHidingPlayers) (Utilities|Array|LastIndex (Variables|Default|GetHidingPlayers)))
        (:Is%20Valid
          ;[1 1 1] Apply GE that gives GA that handles the whole phone booth sequence
          (bind _output_get (Variables|Default|SetExitingPlayer (Utilities|Array|Get(acopy) (Variables|Default|GetHidingPlayers) (Utilities|Array|LastIndex (Variables|Default|GetHidingPlayers)))))
          (if (Networking|IsServer)
            (GameplayEffects|FortApplyGameplayEffectToTarget _returnvalue "/BattlepassS17/QuestActors/Emperor/GE_Athena_EmperorBooth.GE_Athena_EmperorBooth_C" (Ability|GetFortAbilitySystemComponent _output_get) 0 _self_2 _self_2 0.0 (GameplayEffects|MakeFortEffectContext (Ability|GetFortAbilitySystemComponent _output_get) 0 _self_2 _self_2 0))
            ;[1 1 1] Added to fix FORT-310397
            (GameplayEffects|FortApplyGameplayEffectToTarget _returnvalue "/Game/Athena/Items/EnvironmentalItems/HidingProps/Props/GE_Athena_HidingProp_PhantomBooth_ExitingProp.GE_Athena_HidingProp_PhantomBooth_ExitingProp_C" (Ability|GetFortAbilitySystemComponent _output_get) 0 _self_2 _self_2 0.0 (GameplayEffects|MakeFortEffectContext (Ability|GetFortAbilitySystemComponent _output_get) 0 _self_2 _self_2 0))))))
    (:then_1
      ;[1 1 1] Play Blinds Animation. Don't play if already playing (this blocks it from playing when the player is exiting)
      (if (not (Components|Timeline|IsPlaying (Variables|B_HidingProp_EmperorBooth|GetTimeline_0)))
        (CallFunction|BlindsAnim)))))

(event Custom|BlindsAnim
  (bind _staticmeshcomponent (Variables|Debug|GetStaticMeshComponent))
  ;[1 1 1] Animate Blinds Material
  (Rendering|Material|SetCustomPrimitiveDataFloat _staticmeshcomponent 6 1.0)
  (bind (_direction _float) (Timeline|Timeline_0 PlayFromStart)
    (:Update
      (Rendering|Material|SetCustomPrimitiveDataFloat _staticmeshcomponent 5 _float))
    (:Finished
      (Utilities|FlowControl|Delay 1.0)
      (Rendering|Material|SetCustomPrimitiveDataFloat _staticmeshcomponent 6 0.0))))

(fn BlueprintCanInteract
  (params (InteractingPawn :type Fort%20Pawn%20Object%20Reference) (InteractionBeingAttempted :type EInteractionBeingAttempted%20Enum :default "FirstInteraction") (InteractionType :type TInteractionType%20Enum :default "IT_NoInteraction"))
  (return (and (GameplayTags|TagContainer|CompareTagAssetInterfacetoOtherTags (Variables|Default|GetT_OnQuest) InteractingPawn) (|CallParentFunction InteractingPawn))))

(fn BlueprintGetInteractionString
  (params (InteractingPawn :type Fort%20Pawn%20Object%20Reference) (InteractionBeingAttempted :type EInteractionBeingAttempted%20Enum :default "FirstInteraction"))
  (|CallParentFunction InteractingPawn InteractionBeingAttempted)
  (return (select (Interaction|BlueprintCanInteract InteractingPawn) "NSLOCTEXT(\"[9A552FDA48A679CAF41E078597300AD8]\", \"9E334DBE486DD424381002A2BD1F595F\", \"Enter Phone Booth\")" "NSLOCTEXT(\"[9A552FDA48A679CAF41E078597300AD8]\", \"2120BEDF47AFC93E970274A5D1CD93D2\", \"Locked\")")))

(fn ConstructionScript
  (bind _staticmeshcomponent (Variables|Debug|GetStaticMeshComponent))
  (|CallParentFunction)
  ;[1 1 1] Reset Blinds Material
  (Rendering|Material|SetCustomPrimitiveDataFloat _staticmeshcomponent 5 1.0)
  (Rendering|Material|SetCustomPrimitiveDataFloat _staticmeshcomponent 6 0.0)
  (if (Fortnite|IsMobilePlatform)
    (Components|DestroyComponent (Variables|Default|GetSM_Telephone_Booth_GlassWindows))))
