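# Translucent, unlit, two-sided material for a glowing orange ring effect on the Season 17 battle pass Emperor quest actor.
# A curve atlas glint along U, flipped by which side the camera views, drives color and opacity.
# Opacity also combines view-angle, near-camera and optional far fades, plus depth and synced distance fades skipped on low quality.
# SplineThicken supplies world position offset for thickness, pulsed by a time-based sine.
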Material /BattlepassS17/QuestActors/Emperor/FX/M_Emperor_Ring
graph {
 properties {blendMode=BLEND_Translucent, pixelDepthOffsetMode=PDOM_Legacy, shadingModel=MSM_Unlit, twoSided=true, usedShadingModels=Unlit}
 nodes [
  n1 = ComponentMask(None=TextureCoordinate()) {r=true}
  n2 = DotProduct(A=Transform(None=Constant3Vector() {constant=<rgba 0 1 0 0>}) {transformSourceType=TRANSFORMSOURCE_Local}, B=CameraVectorWS())
  n3 = Clamp(None=Multiply(A=@n2, B=ObjectRadius()))
  n4 = CurveAtlasRowParameter(CurveTime=LinearInterpolate(A=@n1, B=OneMinus(None=@n1), Alpha=LinearInterpolate(Alpha=@n3))) {
   atlas=/Game/Athena/LinearColorCurves/VFX_CurveAtlas
   curve=/Game/Effects/Fort_Effects/FrontEnd/SpyBase/Materials/Curve_BattlePass_Door_Glint_Scale
   defaultValue=6
  }
  Reroute(None=LinearInterpolate(B=Constant3Vector() {constant=<rgba 10 3.1401 0 1>}, Alpha=Multiply(A=@n4.G, B=@n4.G)))
  n5 = Constant() {r=1}
  n6 = ScalarParameter() {defaultValue=0.25, parameterName="Fade_Start_Distance (S)"}
  n7 = ObjectRadius()
  n8 = ScalarParameter() {defaultValue=0.05, parameterName="Fade_End_Distance (S)"}
  n9 = DepthFade(
   Opacity=MaterialFunctionCall(Position=WorldPosition().XYZ, Fade_Start_Distance=Multiply(A=@n6, B=@n7), Fade_End_Distance=Multiply(A=@n7, B=@n8)) {materialFunction=&MF_Distance_SyncedFade_FX}.Result
   FadeDistance=ScalarParameter() {defaultValue=300, group=Fade, parameterName=DepthFadeDistance}
  )
  n10 = Multiply(
   A=QualitySwitch(Default=@n5, Low=@n5, High=@n9, Medium=@n9)
   B=MaterialFunctionCall(Distance=ScalarParameter() {defaultValue=1500, group=Fade, parameterName=CameraFadeDisance}) {materialFunction=&Near_Camera_Fade}.Result
  )
  n11 = PixelDepth()
  n12 = ScalarParameter() {group=Fade, parameterName=FarDIstanceFadeStart}
  n13 = ScalarParameter() {defaultValue=10000, group=Fade, parameterName=FarDistanceFadeEnd}
  n14 = ScalarParameter() {group=Fade, parameterName=FarDistanceFadeHardness}
  n15 = Multiply(
   A=@n10
   B=LinearInterpolate(
    A=@n10
    B=SphereMask(A=@n11.R, B=@n12, Radius=@n13, Hardness=@n14) {hardnessPercent=10}
    Alpha=ScalarParameter() {group=Fade, parameterName=UseFarDistanceFade}
   )
  )
  n16 = Abs(None=@n2)
  n17 = Power(
   Base=Subtract(A=LinearInterpolate(Alpha=@n16), B=ScalarParameter() {defaultValue=0.3, parameterName=Subtraction}) {constB=1.25}
   Exp=ScalarParameter() {defaultValue=1, parameterName=FadePower}
  )
  Reroute(None=Multiply(A=Multiply(A=@n15, B=@n4.G), B=Saturate(None=@n17)))
  n18 = ScalarParameter() {defaultValue=150, parameterName=Thickness}
  n19 = ScalarParameter() {defaultValue=2, parameterName=SineWaveSize}
  n20 = Time()
  n21 = Frac(None=ConstantBiasScale(None=Sine(None=@n20)))
  Reroute(None=Multiply(
    A=MaterialFunctionCall(WidthBase=@n18, WidthTip=@n18) {materialFunction=&SplineThicken}.WorldPositionOffset
    B=LinearInterpolate(B=@n19, Alpha=@n21) {constA=1}
   ))
 ]
}
