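; Gameplay ability for the Emperor booth hiding prop that swaps a hidden player's Emperor outfit to its Stage2 variant and reports a quest stat.
; Activation hides HUD, briefly disables input, holsters the weapon and wobbles the booth.
; A timer then triggers booth exit, a show-off montage and backpack override, with velocity tasks watching for movement.
; Tag and prop-death watchers handle interruption; ability end applies cooldown and restores HUD, input and weapon.
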(event Ability|EventFailedToActivatePassiveAbility
  (params (ActorInfo :type Gameplay%20Ability%20Actor%20Info%20Structure))
  (|CallParentFunction))

(event Ability|EventActivateAbility
  (bind _returnvalue (Ability|GetAvatarActorfromActorInfo))
  (bind _returnvalue_2 (Fortnite|GetFortPlayerControllerfromActor _returnvalue))
  (bind _pawn (Variables|Default|GetPawn))
  (bind _asfort_player_pawn (Utilities|Casting|CastToFortPlayerPawn _pawn))
  ;[1 1 1] Hide HUD, save off hiding props and player pawns
  (bind _asb_hiding_prop (Utilities|Casting|CastToB_HidingProp (Ability|EffectContext|GetEffectCauser (Ability|GetGrantedbyEffectContext)))
    (:then
      (Variables|Default|SetHidingProp _asb_hiding_prop)
      (Variables|Default|SetHidingPlayers (Variables|/Game/Athena/Items/EnvironmentalItems/HidingProps/B_HidingProp.B_HidingProp_C|GetHidingPlayers _asb_hiding_prop))
      (CallFunction|WaitforTags)
      (Utilities|IsValid (Utilities|Array|Get(acopy) (Variables|Default|GetHidingPlayers) (Utilities|Array|LastIndex (Variables|Default|GetHidingPlayers)))
        (:Is%20Valid
          (bind _asfort_pawn (Utilities|Casting|CastToFortPawn _returnvalue))
          (bind _output_get (Variables|Default|SetPawn _asfort_pawn))
          (Input|DisableInput _output_get _returnvalue_2)
          (bind _asfort_player_controller_gameplay (Utilities|Casting|CastToFortPlayerControllerGameplay _returnvalue_2))
          (HUD|SetHUDElementVisibilitywithContext _asfort_player_controller_gameplay (Variables|Default|GetT_HudElementToHide) true "None" "")
          (Analytics|FireAnalyticEventEnvironmentalItemUsed _returnvalue_2 (Variables|Default|GetHidingProp))
          (Ability|RemoveGameplayEffectFromOwnerWithGrantedTags "(GameplayTags=((TagName=\"Athena.Quests.EmperorBoothLock\")))")
          (Utilities|FlowControl|Delay 1.0)
          (Input|EnableInput (Ability|GetAvatarActorfromActorInfo) (Fortnite|GetFortPlayerControllerfromActor (Ability|GetAvatarActorfromActorInfo)))
          ;[1 1 1] Play Audio, swap cosmetic, exit booth
          (Quest|SendComplexCustomStatEvent (Quests|GetQuestManager (Fortnite|GetFortPlayerControllerfromActor _asfort_player_pawn) "Invalid") 0 "(GameplayTags=((TagName=\"Athena.Quests.EmperorSwap\")))" "(GameplayTags=)" 1)
          (Audio|PlaySoundatLocation "/BRCosmetics/Sounds/Emotes/Emperor/Emote_Emperor_Intro_1p.Emote_Emperor_Intro_1p" (Transformation|GetActorLocation _asfort_player_pawn))
          (Class|FortPlayerController|ChangeActiveVariantforCosmeticItem (Fortnite|GetFortPlayerControllerfromActor _pawn) "AthenaCharacter:cid_a_109_athena_commando_m_emperor" "(TagName=\"Cosmetics.Variant.Channel.Progressive\")" "(TagName=\"Cosmetics.Variant.Property.Stage2\")" 0)
          (Networking|FlushNetDormancy (Variables|Default|GetHidingProp))
          ;;[1 1 1] Holster weapon to hide pickaxe. (when swapping character parts, player mesh stands up and ignores the animation, so weapon must be hidden manually so it doesn't stick out of the phonebooth)
          (Weapon|SetHolsterWeapon (Variables|Default|GetPawn) 0 true false "None")
          (bind _asb_hiding_prop_emperor_booth (Utilities|Casting|CastToB_HidingProp_EmperorBooth (Variables|Default|GetHidingProp)))
          (Class|BHidingPropEmperorBooth|ChangingOutFitWobble _asb_hiding_prop_emperor_booth)
          (Utilities|Time|SetTimerbyEvent (delegate Custom|OutfitChangeComplete) 1.0))))
    (:CastFailed
      (Ability|EndAbility))))

(event Custom|WaitForTags
  ;[1 1 1] Cancel everything if player gets any of these tags
  (Utilities|FlowControl|Sequence
    (:then_0
      (Ability|Tasks|WaitGameplayTagAdd "(TagName=\"Gameplay.Status.DBNO\")"
        (:Added
          _)))
    (:then_1
      (Ability|Tasks|WaitGameplayTagAdd "(TagName=\"Gameplay.Action.Player.CarryingPlayer\")"
        (:Added
          _)))
    (:then_2
      (Ability|Tasks|WaitGameplayTagAdd "(TagName=\"Athena.Item.Snowman\")"
        (:Added
          _)))
    (:then_3
      (EventDispatchers|BindEventtoOnDied (Variables|Default|GetHidingProp) (delegate Custom|End))))
  (Class|BHidingProp|StopHiding (Variables|Default|GetHidingProp) (Variables|Default|GetPawn))
  (Ability|EndAbility))

(event Custom|OutfitChangeComplete
  (bind _pawn (Variables|Default|GetPawn))
  ;[1 1 1] Play Audio, swap cosmetic, exit booth
  (Class|BHidingProp|EndHidingAnalyticSession (Variables|Default|GetHidingProp) _pawn "PlayerExit")
  (Ability|SendGameplayEventtoActor _pawn "(TagName=\"Event.Athena.HidingProp.StopHiding\")" (Utilities|Struct|MakeGameplayEventData :Instigator (Variables|Default|GetHidingProp) :OptionalObject (Variables|Default|GetHidingProp) :ContextHandle "()" :InstigatorTags "(GameplayTags=)" :TargetTags "(GameplayTags=)" :EventMagnitude (Variables|Default|GetExitMagnitude) :TargetData "()"))
  (Class|FortPlayerPawn|PostResIn (Utilities|Casting|CastToFortPlayerPawn (Variables|Default|GetPawn)))
  (Ability|Tasks|PlayMontageAndWait "None" (Variables|Default|GetShowOffMontage) 1.0 "Into"
    (:then
      (Ability|Tasks|WaitDelay 1.0
        (:OnFinish
          (Cosmetics|PushCosmeticOverrideOntoPawn (Utilities|Casting|CastToFortPlayerPawn (Variables|Default|GetPawn)) "/BRCosmetics/Athena/Items/Cosmetics/Backpacks/BID_775_Emperor.BID_775_Emperor")
          ;;[1 1 1] Cancel Ability if you start moving before anim montage is over
          (Utilities|FlowControl|Sequence
            (:then_0
              (Ability|Tasks|WaitVelocityChange "1.000000,0.000000,0.000000" 225.0
                (:OnVelocityChage
                  _)))
            (:then_1
              (Ability|Tasks|WaitVelocityChange "0.000000,1.000000,0.000000" 225.0
                (:OnVelocityChage
                  _)))
            (:then_2
              (Ability|Tasks|WaitVelocityChange "-1.000000,0.000000,0.000000" 225.0
                (:OnVelocityChage
                  _)))
            (:then_3
              (Ability|Tasks|WaitVelocityChange "0.000000,-1.000000,0.000000" 225.0
                (:OnVelocityChage
                  _)))))))
    (:OnCompleted
      _))
  (Ability|EndAbility))

(event Custom|End
  (params (DamagedActor :type Actor%20Object%20Reference) (Damage :type Float%20%28single-precision%29) (InstigatedBy :type Controller%20Object%20Reference) (DamageCauser :type Actor%20Object%20Reference) (HitLocation :type Vector) (FHitComponent :type Primitive%20Component%20Object%20Reference) (BoneName :type Name) (Momentum :type Vector))
  (Ability|EndAbility))

(event Ability|EventOnEndAbility
  (params (bWasCancelled :type Boolean))
  (bind _returnvalue (Ability|GetAvatarActorfromActorInfo))
  (bind _returnvalue_2 (Fortnite|GetFortPlayerControllerfromActor _returnvalue))
  ;[1 1 1] End Ability, Turn HUD back on. Enable input. UnHolster Weapon
  (Ability|ApplyGameplayEffectToOwner "/Game/Athena/Items/EnvironmentalItems/HidingProps/GE_Athena_HidingProp_PhantomBoothCoolDown.GE_Athena_HidingProp_PhantomBoothCoolDown_C")
  (Input|EnableInput _returnvalue _returnvalue_2)
  (bind _asfort_player_controller_gameplay (Utilities|Casting|CastToFortPlayerControllerGameplay _returnvalue_2)
    (:then
      (HUD|SetHUDElementVisibilitywithContext _asfort_player_controller_gameplay (Variables|Default|GetT_HudElementToHide) false "None" ""))
    (:CastFailed
      _))
  (Weapon|SetHolsterWeapon (Variables|Default|GetPawn) 0 false false "None"))
